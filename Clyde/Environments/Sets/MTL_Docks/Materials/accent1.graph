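# Flat accent material in the MTL_Docks environment set, using the Default Lit shading model.
# It holds only a brown constant color and a scalar constant, with no textures, parameters or visible links to material outputs.
# Usage flags enable it for skeletal meshes and mesh deformers.
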Material /Clyde/Environments/Sets/MTL_Docks/Materials/accent1
graph {
 properties {bUsedWithMeshDeformer=true, bUsedWithSkeletalMesh=true, pixelDepthOffsetMode=PDOM_Legacy, usedShadingModels=Default Lit}
 nodes [
  Constant3Vector() {constant=<rgba 0.2656 0.1433 0.053 0>}
  Constant() {r=0.6}
 ]
}
