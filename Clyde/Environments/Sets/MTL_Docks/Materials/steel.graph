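# Default Lit steel material for the MTL_Docks environment set, flagged for use with skeletal meshes and mesh deformers.
# Its graph holds only two dark brown color constants and one scalar constant, with no textures or parameters.
# No links from these constants to material outputs appear, so their roles in the shading are not evident.
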Material /Clyde/Environments/Sets/MTL_Docks/Materials/steel
graph {
 properties {bUsedWithMeshDeformer=true, bUsedWithSkeletalMesh=true, pixelDepthOffsetMode=PDOM_Legacy, usedShadingModels=Default Lit}
 nodes [
  Constant3Vector() {constant=<rgba 0.0851 0.0498 0.0244 0>}
  Constant3Vector() {constant=<rgba 0.0851 0.0498 0.038 0>}
  Constant() {r=0.6}
 ]
}
